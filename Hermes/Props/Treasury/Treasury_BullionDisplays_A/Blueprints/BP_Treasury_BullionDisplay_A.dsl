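; Treasury bullion display prop that can swap the loot it drops when destroyed according to the current playlist.
; On the authority, BeginPlay waits for playlist data and, if scaling by playlist is enabled, runs PickLootGroupByPlaylist.
; That function matches the map's gameplay tag keys against playlist context tags and writes the paired value to the loot-drop-on-death component.
; The loop never breaks, so the last matching entry wins.
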(event EventBeginPlay
  (|CallParentFunction)
  (Networking|SwitchHasAuthority
    (:Authority
      (Class|FortAsyncActionCurrentPlaylistReady|PlaylistDataReadyAsync
        (:OnReady
          (if (Variables|Default|GetbScaleLootByPlaylist)
            (CallFunction|PickLootGroupbyPlaylist)))))))

(event Collision|EventActorBeginOverlap
  (params (OtherActor :type Actor%20Object%20Reference)))

(event EventTick
  (params (DeltaSeconds :type Float%20%28single-precision%29)))

(fn PickLootGroupByPlaylist
  (bind _keys (Utilities|Map|Keys (Variables|Default|GetLootGroupByPlaylistMap)))
  (bind (_array_element _array_index) (Utilities|Array|ForEachLoopwithBreak _keys)
    (:LoopBody
      (bind _returnvalue (Fortnite|GameplayTags|HasTagfromName (GameplayTags|GetTagName _array_element) (Class|FortGameStateAthena|GetAthenaPlaylistContextTags (Fortnite|Athena|GameState|GetGameStateAthena))))
      (if _returnvalue
        (bind _values (Utilities|Map|Values (Variables|Default|GetLootGroupByPlaylistMap)))
        (Variables|/Game/Athena/Items/Components/BuildingActorLootDropOnDeathComponent.BuildingActorLootDropOnDeathComponent_C|SetLootOnExplodePackage (Utilities|Array|Get(acopy) _values _array_index) (Variables|Default|GetBuildingActorLootDropOnDeathComponent))))))

(fn ConstructionScript)
